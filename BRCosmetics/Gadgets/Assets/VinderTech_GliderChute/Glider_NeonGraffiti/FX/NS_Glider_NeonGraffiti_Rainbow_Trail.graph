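# Niagara system for the Neon Graffiti glider's rainbow trail, drawing two bone-anchored ribbon trails while the glider is fully deployed.
# One AnimTrail emitter initializes short-lived particles from two bone or socket pairs, then colors them with a hue-shifted blue-to-red gradient.
# Ribbon alpha follows a system-level integrator through a curve.
# The front-end flag instead forces full alpha and adds velocity along negative Y, so the trail streams in menus.
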NiagaraSystem /BRCosmetics/Gadgets/Assets/VinderTech_GliderChute/Glider_NeonGraffiti/FX/NS_Glider_NeonGraffiti_Rainbow_Trail
dependencies {
 usedDataInterfaces [&NiagaraDataInterfaceColorCurve, &NiagaraDataInterfaceCurve]
 usedDynamicInputs [
  &Boolean_OrOperation
  &SetBoolByFloatComparison
  &Abs_Float
  &Boolean_AndOperation
 ]
 usedModules [&SystemState, &FN_Integrator]
}
emitters {FN_AnimTrail {
  emitterSpawnScript [&FN_RibbonIDOffsetUtility()]
  emitterUpdateScript [
   &EmitterState()
   &FN_SpawnAnimTrail(Enable:NiagaraBool=$User.bisFullyDeployed, Number of Anim Trails=2)
  ]
  particleSpawnScript [
   &FN_InitializeAnimTrailBoneIndecies(First Bone or Socket Index=1, Second Bone or Socket Index=0)
   &FN_InitializeAnimTrailBoneIndecies(AnimTrail Selection=1, First Bone or Socket Index=3, Second Bone or Socket Index=2) {module=FN_InitializeAnimTrailBoneIndecies001}
   &FN_InitializeAnimTrail(Lifetime=0.2, Write Material Random=false)
   &AddVelocity(Velocity=&Lerp_Vector(
     A=<0 0 0>
     B=<0 -1150 0>
     Alpha=&MakeCustomFloatFromBool(Bool=$User.bisFrontEnd, True Float=1, False Float=0)
    ))
  ]
  particleUpdateScript [
   &SolveForcesAndVelocity(Clamp Velocity=false, Limit Acceleration=false)
   /Game/Effects/Niagara/Modules/AnimTrail/FN_UpdateAnimTrail/FN_UpdateAnimTrail(New float=1)
   /Game/Effects/Niagara/Modules/Color/FN_ColorHueShift(
    Color:LinearColor=&MakeLinearColorFromVectorAndFloat(
     "Vector (RGB)"=&MakeVectorFromLinearColorRGB(Color=&ColorFromCurve(ColorCurve=DataInterface() {propertyValues {
         AlphaCurve {keys [<at 0 1 RCIM_Cubic>]}
         BlueCurve {keys [<at 0.181 1 RCIM_Cubic>,<at 0.4605 1>,<at 0.7452 0.6167>]}
         GreenCurve {keys [<at 0.181 0.4 RCIM_Cubic>,<at 0.4605 0>,<at 0.7452 0>]}
         RedCurve {keys [<at 0.181 0 RCIM_Cubic>,<at 0.4605 0.0167>,<at 0.7452 1>]}
        }}))
     "Float (Alpha)"=&FloatFromCurve(
      FloatCurve=DataInterface() {propertyValues {Curve {keys [<at 0 0 RCIM_Cubic>, {arriveTangent=2.0248, interpMode=RCIM_Cubic, leaveTangent=2.0248, tangentMode=RCTM_User, time=1, value=1}]}}}
      CurveIndex=&MakeCustomFloatFromBool(Bool=$User.bisFrontEnd, True Float=1, False Float=$System.FN_Integrator.Value)
     )
    )
    Initial Offset=0
    Speed=0.5
   )
  ]
  rendererClasses [&NiagaraRibbonRendererProperties]
  renderers [{rendererClass=&NiagaraRibbonRendererProperties}]
 }}
summary {
 systemName=NS_Glider_NeonGraffiti_Rainbow_Trail
 emitters [{emitterName=FN_AnimTrail, rendererClasses [&NiagaraRibbonRendererProperties]}]
 userVariables [
  User.bisFrontEnd {type=NiagaraBool}
  User.bisFullyDeployed {type=NiagaraBool}
  User.ForwardDot {type=NiagaraFloat, defaultValue=0, description=""}
  User.RibbonIDOffset {type=NiagaraInt32}
  User.RightDot {type=NiagaraFloat, defaultValue=0, description=""}
  User.SkeletalMesh {
   type=NiagaraDataInterfaceSkeletalMesh
   defaultValue {
    struct=NiagaraExt_VariableValue_DataInterface
    value {dataInterface=&self:NiagaraDataInterfaceSkeletalMesh_0}
   }
  }
 ]
}
